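# Physics asset for the hair and bangs face accessory of the female medium Textile Ram player body, built entirely from capsule bodies.
# Two single-capsule strands (dyn_hair_02, dyn_hair_03) and a tapering four-body dyn_hair_05 chain hang from the head, with tighter twist than swing limits.
# Head, faceAttach, neck_01 and spine_05 capsules give the body volume around the hair, linked by constraints with free swing and twist.
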PhysicsAsset /BRCosmetics/Characters/Player/Female/Medium/Bodies/F_MED_Textile_Ram/Meshes/Parts/F_MED_Textile_Ram_FaceAcc_HairAndBangs_Physics
bodies {
 dyn_hair_02 [{shapeType=Capsule, length=5.8779, radius=0.9924, shapeName=dyn_hair_02_capsule, center=<3.8475 0.0001 -0.999>, rotation=<rot 71.8601 0 0>}]
 dyn_hair_03 [{shapeType=Capsule, length=5.9633, radius=0.907, shapeName=dyn_hair_03_capsule, center=<3.6994 0 -0.5073>, rotation=<rot 73.7255 0 0>}]
 dyn_hair_05 [{shapeType=Capsule, length=2.25, radius=3.5, center=<0.6218 0 0.2692>, rotation=<rot 66.5908 -180 -180>}]
 dyn_hair_05_a [{shapeType=Capsule, length=3, radius=2.75, center=<1.1885 0.0001 0>, rotation=<rot 90 0 0>}]
 dyn_hair_05_b [{shapeType=Capsule, length=2.75, radius=2.5, center=<0.9382 0 0>, rotation=<rot 90 0 0>}]
 dyn_hair_05_c [{shapeType=Capsule, length=2.5, radius=2, center=<1.0002 0 0>, rotation=<rot 90 0 0>}]
 faceAttach [{shapeType=Capsule, length=8.6271, radius=5.2559, center=<2.2974 6.072 2.3583>, rotation=<rot 8.2948 93.3576 -46.839>}]
 head [{shapeType=Capsule, length=7.5921, radius=6.6484, center=<0.1705 0.1239 -0.432>, rotation=<rot 0 79.873 -89.9998>}]
 neck_01 [{shapeType=Capsule, length=4.5, radius=4.505, center=<0 0.2774 -0.0001>, rotation=<rot 0 -93.282 -89.9999>}]
 spine_05 [{shapeType=Capsule, length=9, radius=11.75, center=<0.0777 0.5063 0>}]
}
constraints [
 {
  bone1Name=faceAttach
  bone2Name=head
  swing1LimitDegrees=45
  swing1Motion=Free
  swing2LimitDegrees=45
  swing2Motion=Free
  twistLimitDegrees=45
  twistMotion=Free
 }
 {bone1Name=dyn_hair_05, bone2Name=head, swing1LimitDegrees=45, swing2LimitDegrees=45, twistLimitDegrees=5}
 {
  bone1Name=head
  bone2Name=neck_01
  swing1LimitDegrees=45
  swing1Motion=Free
  swing2LimitDegrees=45
  swing2Motion=Free
  twistLimitDegrees=45
  twistMotion=Free
 }
 {
  bone1Name=neck_01
  bone2Name=spine_05
  swing1LimitDegrees=45
  swing1Motion=Free
  swing2LimitDegrees=45
  swing2Motion=Free
  twistLimitDegrees=45
  twistMotion=Free
 }
 {bone1Name=dyn_hair_05_a, bone2Name=dyn_hair_05, swing1LimitDegrees=45, swing2LimitDegrees=45, twistLimitDegrees=5}
 {bone1Name=dyn_hair_05_b, bone2Name=dyn_hair_05_a, swing1LimitDegrees=45, swing2LimitDegrees=45, twistLimitDegrees=5}
 {bone1Name=dyn_hair_05_c, bone2Name=dyn_hair_05_b, swing1LimitDegrees=45, swing2LimitDegrees=45, twistLimitDegrees=5}
 {bone1Name=dyn_hair_02, bone2Name=head, swing1LimitDegrees=45, swing2LimitDegrees=45, twistLimitDegrees=5}
 {bone1Name=dyn_hair_03, bone2Name=head, swing1LimitDegrees=45, swing2LimitDegrees=45, twistLimitDegrees=5}
]
