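# Anim Blueprint for the F_MED_Textile_Ram face accessory, copying the parent mesh pose and adding secondary hair motion.
# Front end uses RigidBody physics; otherwise bools select head shrinking for custom-part playlists, plain copied pose, RigidBody or an AnimDynamics fallback on dyn_hair_05 bones.
# The result gets a sunglasses offset, plus head constraints and pitch on two hair bones when rigid body is disabled.
# Several AnimDynamics chains are left unconnected.
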AnimBlueprint /BRCosmetics/Characters/Player/Female/Medium/Bodies/F_MED_Textile_Ram/Meshes/Parts/F_MED_Textile_Ram_FaceAcc_AnimBP
target_skeleton=&/F_MED_Textile_Ram_FaceAcc_Skeleton
graphs [
 AnimGraph {nodes [
   n1 = Anim Dynamics - Bone: dyn_hair_03(Alpha=1, ExternalForce="0,0,0") {
    angularDampingOverride=1
    bAngularSpring=true
    bGravityOverrideInSimSpace=true
    bOverrideAngularDamping=true
    numSolverIterationsPostUpdate=4
    numSolverIterationsPreUpdate=16
    alphaBoolBlend {blendOption=Linear}
    boundBone {boneName=dyn_hair_03}
    componentAppliedLinearAccClamp=<5000 5000 5000>
    physicsBodyDefinitions [{
      sphereCollisionRadius=0
      boundBone {boneName=dyn_hair_03}
      boxExtents=<30 30 10>
      constraintSetup {angularLimitsMax=<45 45 45>, angularLimitsMin=<-45 -45 -45>, angularTarget=<0 0 1>}
      localJointOffset=<0.2 0 -10>
     }]
   }
   n2 = Anim Dynamics - Bone: dyn_hair_02(ComponentPose=@n1, Alpha=1) {
    angularDampingOverride=1
    bAngularSpring=true
    bGravityOverrideInSimSpace=true
    bOverrideAngularDamping=true
    numSolverIterationsPostUpdate=4
    numSolverIterationsPreUpdate=16
    alphaBoolBlend {blendOption=Linear}
    boundBone {boneName=dyn_hair_02}
    componentAppliedLinearAccClamp=<5000 5000 5000>
    physicsBodyDefinitions [{
      sphereCollisionRadius=0
      boundBone {boneName=dyn_hair_02}
      boxExtents=<30 20 10>
      constraintSetup {angularLimitsMax=<45 45 45>, angularLimitsMin=<-45 -45 -45>, angularTarget=<0 0 1>}
      localJointOffset=<0.2 0 -15>
     }]
   }
   Anim Dynamics - Bone: dyn_hair_04(ComponentPose=@n2, Alpha=0, ExternalForce="0,0,0") {
    angularDampingOverride=1
    angularSpringConstant=250
    bAngularSpring=true
    bGravityOverrideInSimSpace=true
    bOverrideAngularDamping=true
    gravityScale=0.001
    numSolverIterationsPostUpdate=4
    numSolverIterationsPreUpdate=16
    alphaBoolBlend {blendOption=Linear}
    boundBone {boneName=dyn_hair_04}
    componentAppliedLinearAccClamp=<5000 5000 5000>
    physicsBodyDefinitions [{
      sphereCollisionRadius=0
      boundBone {boneName=dyn_hair_04}
      boxExtents=<30 20 10>
      constraintSetup {
       angularTargetAxis=AxisY
       twistAxis=AxisY
       angularLimitsMax=<20 20 45>
       angularLimitsMin=<-20 -20 -45>
       angularTarget=<0 0 1>
       linearAxesMax=<0.5 0.5 0.5>
       linearAxesMin=<-0.5 -0.5 -0.5>
      }
      localJointOffset=<0 0 0>
     }]
   }
   n3 = Anim Dynamics - Bone: dyn_hair_03(Alpha=1, ExternalForce="0,0,0") {
    angularDampingOverride=0.9999
    angularSpringConstant=50
    bAngularSpring=true
    bGravityOverrideInSimSpace=true
    bOverrideAngularDamping=true
    simulationSpace=RootRelative
    alphaBoolBlend {blendOption=Linear}
    boundBone {boneName=dyn_hair_03}
    componentLinearAccScale=<1 1 1>
    componentLinearVelScale=<1 1 1>
    physicsBodyDefinitions [{
      sphereCollisionRadius=0
      boundBone {boneName=dyn_hair_03}
      boxExtents=<20 20 10>
      constraintSetup {angularLimitsMax=<45 45 45>, angularLimitsMin=<-45 -45 -45>, angularTarget=<0 0 0>}
      localJointOffset=<-0.5 0 -17>
     }]
   }
   Anim Dynamics - Bone: dyn_hair_02(ComponentPose=@n3, Alpha=1) {
    angularDampingOverride=0.9999
    angularSpringConstant=50
    bAngularSpring=true
    bGravityOverrideInSimSpace=true
    bOverrideAngularDamping=true
    simulationSpace=RootRelative
    alphaBoolBlend {blendOption=Linear}
    boundBone {boneName=dyn_hair_02}
    componentLinearAccScale=<1 1 1>
    componentLinearVelScale=<1 1 1>
    physicsBodyDefinitions [{
      sphereCollisionRadius=0
      boundBone {boneName=dyn_hair_02}
      boxExtents=<20 20 10>
      constraintSetup {angularLimitsMax=<45 45 45>, angularLimitsMin=<-45 -45 -45>, angularTarget=<0 0 0>}
      localJointOffset=<-0.5 0 -17>
     }]
   }
   Anim Dynamics - Bone: dyn_hair_03(Alpha=1, ExternalForce="0,0,0") {
    angularDampingOverride=1
    angularSpringConstant=250
    bAngularSpring=true
    bGravityOverrideInSimSpace=true
    bOverrideAngularDamping=true
    numSolverIterationsPostUpdate=4
    numSolverIterationsPreUpdate=16
    alphaBoolBlend {blendOption=Linear}
    boundBone {boneName=dyn_hair_03}
    componentAppliedLinearAccClamp=<5000 5000 5000>
    physicsBodyDefinitions [{
      sphereCollisionRadius=0
      boundBone {boneName=dyn_hair_03}
      boxExtents=<30 30 10>
      constraintSetup {angularLimitsMax=<10 35 0>, angularLimitsMin=<-20 -5 0>, angularTarget=<0 0 1>}
      localJointOffset=<1.5 0 -17>
     }]
   }
   n4 = Get HeadRotationBack() -> (HeadRotationBack:double)
   n5 = "Transform (Modify) Bone - Bone: dyn_hair_05"(
    ComponentPose=Local To Component(LocalPose=Copy Pose From Mesh() {bCopyCurves=true, bUseAttachedParent=true})
    Alpha=@n4
    Translation="0,0,0"
    Rotation="15,0.000,0.000"
    Scale="1,1,1"
   ) {
    rotationMode=BMM_Additive
    rotationSpace=BCS_BoneSpace
    translationMode=BMM_Additive
    translationSpace=BCS_BoneSpace
    alphaBoolBlend {blendOption=Linear}
    boneToModify {boneName=dyn_hair_05}
    rotation {pitch=15}
   }
   n6 = "Transform (Modify) Bone - Bone: dyn_hair_05_a"(ComponentPose=@n5, Alpha=@n4, Translation="0,0,0", Rotation="25,0.000,0.000", Scale="1,1,1") {
    rotationMode=BMM_Additive
    rotationSpace=BCS_BoneSpace
    translationMode=BMM_Additive
    translationSpace=BCS_BoneSpace
    alphaBoolBlend {blendOption=Linear}
    boneToModify {boneName=dyn_hair_05_a}
    rotation {pitch=25}
   }
   n7 = Anim Dynamics - Bone: dyn_hair_05_a(ComponentPose=@n6, Alpha=1, ExternalForce="0,0,0") {
    angularDampingOverride=0.9999
    angularSpringConstant=1500
    bGravityOverrideInSimSpace=true
    bOverrideAngularDamping=true
    bUseGravityOverride=true
    bUseSphericalLimits=true
    simulationSpace=RootRelative
    alphaBoolBlend {blendOption=Linear}
    boundBone {boneName=dyn_hair_05_a}
    componentAppliedLinearAccClamp=<1100 1100 1100>
    componentLinearAccScale=<0.5 0.5 0.5>
    componentLinearVelScale=<0.5 0.5 0.5>
    gravityOverride {z=-980}
    physicsBodyDefinitions [{
      collisionType=CustomSphere
      sphereCollisionRadius=2
      boundBone {boneName=dyn_hair_05_a}
      boxExtents=<8 15 15>
      constraintSetup {angularLimitsMax=<5 30 30>, angularLimitsMin=<-5 -30 -30>, angularTarget=<0 0 1>}
      localJointOffset=<5 0 0>
     }]
    sphericalLimits [{limitRadius=35, drivingBone {boneName=spine_05}, sphereLocalOffset=<-18 15 0>}]
   }
   n8 = Anim Dynamics - Bone: dyn_hair_05_b(ComponentPose=@n7, Alpha=1, ExternalForce="0,0,0") {
    angularDampingOverride=0.9999
    angularSpringConstant=1500
    bGravityOverrideInSimSpace=true
    bOverrideAngularDamping=true
    bUseGravityOverride=true
    bUseSphericalLimits=true
    simulationSpace=RootRelative
    alphaBoolBlend {blendOption=Linear}
    boundBone {boneName=dyn_hair_05_b}
    componentAppliedLinearAccClamp=<1100 1100 1100>
    componentLinearAccScale=<0.5 0.5 0.5>
    componentLinearVelScale=<0.5 0.5 0.5>
    gravityOverride {z=-980}
    physicsBodyDefinitions [{
      collisionType=CustomSphere
      sphereCollisionRadius=1.5
      boundBone {boneName=dyn_hair_05_b}
      boxExtents=<8 15 15>
      constraintSetup {angularLimitsMax=<-5 30 30>, angularLimitsMin=<-5 -30 -30>, angularTarget=<0 0 1>}
      localJointOffset=<2.5 0 0>
     }]
    sphericalLimits [{limitRadius=35, drivingBone {boneName=spine_05}, sphereLocalOffset=<-18 15 0>}]
   }
   n9 = RigidBody(ComponentPose=Local To Component(LocalPose=Copy Pose From Mesh() {bCopyCurves=true, bUseAttachedParent=true})) {
    bClampLinearTranslationLimitToRefPose=true
    bOverrideWorldGravity=true
    overridePhysicsAsset=&/F_MED_Textile_Ram_FaceAcc_HairAndBangs_Physics
    simulationSpace=BaseBoneSpace
    alphaBoolBlend {blendOption=Linear}
    baseBoneRef {boneName=root}
    componentAppliedLinearAccClamp=<500 500 2000>
    componentLinearAccScale=<0.2 0.3 1>
    componentLinearVelScale=<0.2 0.3 1>
    overrideWorldGravity {z=-980}
   }
   n10 = Local To Component(LocalPose=Blend Poses by bool(BlendPose_0=Component To Local(ComponentPose=@n9), BlendPose_1=Component To Local(ComponentPose=@n8), bActiveValue=Get bRigidBodyEnabled()) {blendType=Linear})
   n11 = "Transform (Modify) Bone - Bone: head"!(ComponentPose=Local To Component(LocalPose=Copy Pose From Mesh() {bCopyCurves=true, bUseAttachedParent=true}), Alpha=1) -> (Pose) {scaleMode=BMM_Replace, alphaBoolBlend {blendOption=Linear}, boneToModify {boneName=head}, scale=<0.01 0.01 0.01>}
   n12 = Blend Poses by bool(
    BlendPose_0=Component To Local(ComponentPose=@n11)
    BlendPose_1=Blend Poses by bool(
     BlendPose_0=Copy Pose From Mesh() {bCopyCurves=true, bUseAttachedParent=true}
     BlendPose_1=Component To Local(ComponentPose=@n10)
     bActiveValue=Get bCanPlayCustomAnimations() -> (bCanPlayCustomAnimations:bool)
    ) {blendType=Linear}
    BlendTime_0=0
    BlendTime_1=0
    bActiveValue=Get GetPlayerListUsesCustomCharacterParts() -> (GetPlayerListUsesCustomCharacterParts:bool)
   ) {blendType=Linear}
   n13 = RigidBody(ComponentPose=Local To Component(LocalPose=Copy Pose From Mesh() {bCopyCurves=true, bUseAttachedParent=true})) {
    bClampLinearTranslationLimitToRefPose=true
    bOverrideWorldGravity=true
    overridePhysicsAsset=&F_MED_Textile_Ram_FaceAcc_HairAndBangs_Physics
    simulationSpace=BaseBoneSpace
    alphaBoolBlend {blendOption=Linear}
    baseBoneRef {boneName=root}
    componentAppliedLinearAccClamp=<500 500 1000>
    componentLinearAccScale=<0.2 0.3 1>
    componentLinearVelScale=<0.2 0.3 1>
    overrideWorldGravity {z=-980}
   }
   n14 = Component To Local(ComponentPose=@n13)
   n15 = Get bIsInFrontEnd()
   n16 = "Transform (Modify) Bone - Bone: dyn_sunglasses"(
    ComponentPose=Local To Component(LocalPose=Blend Poses by bool(BlendPose_0=@n14, BlendPose_1=@n12, BlendTime_0=0, BlendTime_1=0, bActiveValue=@n15) {blendType=Linear})
    Alpha=1
    Translation="0,0.35,0.2"
    Rotation="0,0,0"
    Scale="1,1.025,1"
   ) {
    rotationMode=BMM_Additive
    rotationSpace=BCS_BoneSpace
    scaleMode=BMM_Additive
    scaleSpace=BCS_BoneSpace
    translationMode=BMM_Additive
    translationSpace=BCS_BoneSpace
    alphaBoolBlend {blendOption=Linear}
    boneToModify {boneName=dyn_sunglasses}
    scale {y=1.025}
    translation {y=0.35, z=0.2}
   }
   n17 = Get RBAN() -> (RBAN:double)
   n18 = "Constraint - dyn_hair_02 (P)"(ComponentPose=@n16, Alpha=@n17, ConstraintWeights_0=1) {
    alphaBoolBlend {blendOption=Linear}
    boneToModify {boneName=dyn_hair_02}
    constraintSetup [{targetBone {boneName=head}}]
    constraintWeights [1]
   }
   n19 = "Constraint - dyn_hair_03 (P)"(ComponentPose=@n18, Alpha=@n17, ConstraintWeights_0=1) {
    alphaBoolBlend {blendOption=Linear}
    boneToModify {boneName=dyn_hair_03}
    constraintSetup [{targetBone {boneName=head}}]
    constraintWeights [1]
   }
   n20 = "Transform (Modify) Bone - Bone: dyn_hair_02"(ComponentPose=@n19, Alpha=@n17, Translation="0,0,0", Rotation="-20,0,0", Scale="1,1,1") {
    rotationMode=BMM_Additive
    rotationSpace=BCS_BoneSpace
    scaleMode=BMM_Additive
    scaleSpace=BCS_BoneSpace
    translationMode=BMM_Additive
    translationSpace=BCS_BoneSpace
    alphaBoolBlend {blendOption=Linear}
    boneToModify {boneName=dyn_hair_02}
    rotation {pitch=-20}
   }
   n21 = "Transform (Modify) Bone - Bone: dyn_hair_03"(ComponentPose=@n20, Alpha=@n17, Translation="0,0,0", Rotation="-20,0,0", Scale="1,1,1") {
    rotationMode=BMM_Additive
    rotationSpace=BCS_BoneSpace
    scaleMode=BMM_Additive
    scaleSpace=BCS_BoneSpace
    translationMode=BMM_Additive
    translationSpace=BCS_BoneSpace
    alphaBoolBlend {blendOption=Linear}
    boneToModify {boneName=dyn_hair_03}
    rotation {pitch=-20}
   }
   Output Pose(Result=Component To Local(ComponentPose=@n21)) {layerGroup=None}
   Save cached pose 'PreScale Cache'(Pose=Use cached pose 'Sub Graph Cache'()) {cacheName=PreScale Cache}
   Save cached pose 'Sub Graph Cache'() {cacheName=Sub Graph Cache}
   Use cached pose 'PreScale Cache'()
  ]}
 EventGraph {nodes [
   Event BlueprintUpdateAnimation()
   > n1 = Branch(Condition=Get bRigidBodyEnabled())
   > Set RBAN(RBAN:double=0) -> (Output_Get:double)
   Set RBAN(execute=@n1.else, RBAN=1)
   Event BlueprintInitializeAnimation()
   > Set MeshToCopy(MeshToCopy:Skeletal Mesh Component Object Reference=GetOwnerSkeletalMeshComponent()) -> (Output_Get:Skeletal Mesh Component Object Reference)
   > Set GetPlayerListUsesCustomCharacterParts(GetPlayerListUsesCustomCharacterParts:bool=GetPlaylistUsesCustomCharacterParts!() -> (ReturnValue:bool)) -> (Output_Get:bool)
  ]}
]
variables [
 GetPlayerListUsesCustomCharacterParts {type=boolean}
 HeadRotationBack {type=number}
 MeshToCopy {type=SkeletalMeshComponent}
 RBAN {type=number}
]
